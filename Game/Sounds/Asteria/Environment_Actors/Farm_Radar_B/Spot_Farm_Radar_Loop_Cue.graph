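# Looping ambient spot cue for the Asteria Farm Radar B environment actor, classed as outdoor ambience and limited to one concurrent instance.
# A close layer (pitched-up electrical box, comms computers, pad, windmill wind) mixes with a distant layer of three radar dish glow loops, two pitched down.
# Each layer has its own attenuation node borrowing BattleBus siren settings, alongside cue-level distance and low-pass overrides.
# Only the first two quality-level slots are connected, both to the same mix.
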SoundCue /Game/Sounds/Asteria/Environment_Actors/Farm_Radar_B/Spot_Farm_Radar_Loop_Cue
root=QualityLevel_0
nodes [
 {
  id=Attenuation_0
  children [Modulator_2]
  properties {attenuationSettings=/Game/Sounds/Asteria/POI/SpawnIsland/Spots/Spot_BattleBus_Siren_Loop_Close}
 }
 {
  id=Attenuation_1
  children [Mixer_0]
  properties {attenuationSettings=/Game/Sounds/Asteria/POI/SpawnIsland/Spots/Spot_BattleBus_Siren_Loop_Distant_Att}
 }
 {id=Mixer_0, children [WavePlayer_1,Modulator_0,Modulator_1], properties {inputVolume [1,1.65,0.55]}}
 {id=Mixer_1, children [Modulator_3,WavePlayer_4,WavePlayer_5,WavePlayer_6], properties {inputVolume [1,0.5,1,2]}}
 {id=Mixer_2, children [Attenuation_0, Attenuation_1], properties {inputVolume [1,1]}}
 {id=Modulator_0, children [WavePlayer_2], properties {pitchMax=0.8, pitchMin=0.8, volumeMax=1, volumeMin=1}}
 {id=Modulator_1, children [WavePlayer_3], properties {pitchMax=0.8, pitchMin=0.8, volumeMax=1, volumeMin=1}}
 {id=Modulator_2, children [Mixer_1], properties {volumeMax=1, volumeMin=1}}
 {id=Modulator_3, children [WavePlayer_0], properties {pitchMax=1.2, pitchMin=1.2, volumeMax=1, volumeMin=1}}
 {id=QualityLevel_0, children [Mixer_2,Mixer_2,null,null]}
 {
  id=WavePlayer_0
  properties {bLooping=true, soundWaveAssetPtr=/Game/Sounds/Apollo/Biomes/RadarBases/Amb_RadarBase_Tinfoil_ElectricalBox_Loop_01}
 }
 {id=WavePlayer_1, properties {bLooping=true, soundWaveAssetPtr=/Game/Sounds/Apollo/Biomes/RadarBases/Amb_RadarDish_Glow_Loop_01}}
 {id=WavePlayer_2, properties {bLooping=true, soundWaveAssetPtr=/Game/Sounds/Apollo/Biomes/RadarBases/Amb_RadarDish_Glow_Loop_03}}
 {id=WavePlayer_3, properties {bLooping=true, soundWaveAssetPtr=/Game/Sounds/Apollo/Biomes/RadarBases/Amb_RadarDish_Glow_Loop_02}}
 {
  id=WavePlayer_4
  properties {bLooping=true, soundWaveAssetPtr=/Game/Sounds/Apollo/Biomes/GovernmentSecretComplex/S17_more_comms_computers_010_01}
 }
 {id=WavePlayer_5, properties {bLooping=true, soundWaveAssetPtr=/Game/Athena/Sounds/Ambient/SLC/Wavs/SLC_Pad_Loop}}
 {
  id=WavePlayer_6
  properties {bLooping=true, soundWaveAssetPtr=/Game/Sounds/Artemis/Biomes/Grasslands/Grassland_Spots/Spot_Outpost_Windmill_Spin_Close_Wind_Loop}
 }
]
properties {
 bOverrideAttenuation=true
 bOverrideConcurrency=true
 maxDistance=6000
 soundClassObject=&Fort_Ambient_Outdoors_Athena
 volumeMultiplier=0.376
 attenuationOverrides {
  bAttenuateWithLPF=true
  bEnableReverbSend=false
  falloffDistance=3400
  lPFFrequencyAtMax=500
  lPFRadiusMax=5000
  lPFRadiusMin=2000
  stereoSpread=5000
  attenuationShapeExtents {x=500}
 }
 concurrencyOverrides {maxCount=1}
}
